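; Unstable TNT grenade projectile that explodes on every bounce or water entry until it reaches a scalable bounce limit.
; Authority sets bounce velocity: the first picks a random heading outside a rear deadzone, later ones aim at random points around the first impact.
; OnExploded fires a GameplayCue and kills the grenade on the final blast.
; Clients play surface-aware FX, sounds, shake and rumble, hiding mesh and fuse on the last.
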(event EventBeginPlay
  ; BeingPlay
  (|CallParentFunction)
  (Variables|Default|SetNumberOfBouncesTillExplode (Math|Float|Truncate (ScalableFloat|GetValueatLevel (Variables|Default|GetBouncesBeforeExplode))))
  (Variables|Default|SetBounceHeight (ScalableFloat|GetValueatLevel (Variables|Default|GetBouncesHeight)))
  (Variables|Default|SetBounceRadius (ScalableFloat|GetValueatLevel (Variables|Default|GetBouncesRadius)))
  (Variables|Default|SetBounce%20Force (ScalableFloat|GetValueatLevel (Variables|Default|GetBouncesForce)))
  (Variables|Default|SetFirstBounceDeadzoneHalfAngle (ScalableFloat|GetValueatLevel (Variables|Default|GetFirstBounceDeadzoneAngle))))

(event EventOnExploded
  (params (HitActors :type Array%20of%20Actor%20Object%20References) (HitResults :type Array%20of%20Hit%20Result%20Structures))
  (bind _returnvalue (Transformation|GetActorLocation))
  (bind _returnvalue_2 (Game|GetInstigator))
  (bind _e_surface_type (Variables|Default|GetE_Surface_Type))
  (bind _returnvalue_3 (Math|Conversions|ToInteger(Byte) _e_surface_type))
  (bind _cue_explosionsound (Variables|Sounds|GetCue_ExplosionSound))
  (bind _cuebounceexplosionsound (Variables|Sounds|GetCueBounceExplosionSound))
  ;[1 1 1] Bounce Explosions
  (Networking|SwitchHasAuthority
    (:Authority
      (GameplayCue|ExecuteGameplayCueOnActor(Burst) self (Variables|Default|GetGC_OnExplode) (Ability|GameplayCue|MakeGameplayCueParameters (select (Variables|Default|GetFinalBounce) 1.0 0.0)))
      (if (Variables|Default|GetFinalBounce)
        (Delivery|Kill)))
    (:Remote
      (Networking|CanExecuteCosmeticEvents
        (:True
          (bind _returnvalue_4 (Effects|Components|ParticleSystem|SpawnFXAtLocation (CallFunction|GetExplosion (or (Water|IsinWater (Variables|Water|GetWaterInteractionComponent)) (Utilities|Enum|Equal(Enum) (Variables|Default|GetE_Surface_Type) "SurfaceType8"))) (Transformation|GetActorLocation) (Variables|Default|GetExplosion%20Rotation)))
          ;[1 1 1] PhysSurface Interaction for explosion Niagara FX
          (bind _checkedphysicalsurfaceid (CallFunction|PhysicalSurfaceCheck _returnvalue_3))
          (if _checkedphysicalsurfaceid
            (bind _returnvalue_5 (Effects|Components|ParticleSystem|SpawnFXAtLocation (Variables|VFX|GetFX_Physical%20Surface)))
            (Effects|Components|ParticleSystem|SetIntParameter _returnvalue_5 "PhysicalSurface" _returnvalue_3)
            (Effects|Components|ParticleSystem|SetFloatParameter _returnvalue_5 "Global Scale" 0.8))
          (Rendering|SetRenderCustomDepth _returnvalue_4 true) ; DISABLED
          (Utilities|FlowControl|Sequence
            (:then_0
              (Camera|PlayWorldCameraShake (Variables|Default|GetExplosionCameraShake) (Transformation|GetActorLocation) 200 800))
            (:then_1
              (if (Variables|Default|GetFinalBounce)
                (Utilities|IsValid _cue_explosionsound
                  (:Is%20Valid
                    (Audio|PlaySoundatLocation _cue_explosionsound _returnvalue "0, 0, 0" 1.0 1.0 0.0 0 0 _returnvalue_2) ; Play Explosion
                    (Fortnite|ShowSoundIndicatorLocal :Location (Transformation|GetActorLocation) :MaxAudibleDistance (Variables|/Script/Engine.SoundBase|GetMaxDistance (Variables|Sounds|GetCue_ExplosionSound)) :IndicatorType "Gunshot" :Instigator (Variables|Actor|GetInstigator) :Tint "(R=0.000000,G=0.000000,B=0.000000,A=0.000000)" :SoundBase (Variables|Sounds|GetCue_ExplosionSound) :AttenuationMaxRangeMultiplier 1.0 :AttenuationIntensityMultiplier 1.0 :!label _label) ; Indicator for explosion
                  )
                  (:Is%20Not%20Valid
                    _))
                (else
                  (Utilities|IsValid _cuebounceexplosionsound
                    (:Is%20Valid
                      (Audio|PlaySoundatLocation _cuebounceexplosionsound _returnvalue "0, 0, 0" 1.0 1.0 0.0 0 0 _returnvalue_2) ; Play Explosion
                      (goto _label))
                    (:Is%20Not%20Valid
                      _))))
              (bind _waterfuseaudiocomponent (Variables|Default|GetWaterFuseAudioComponent)
                (:then
                  (Audio|Components|Audio|FadeOut _waterfuseaudiocomponent 0.2 0.0 "Linear" :!class /Script/Engine.AudioComponent))
                (:else)))
            (:then_2
              (Fortnite|PlayLocalForceFeedbackatLocationMulti (Variables|Default|GetExplosionForceFeedbackNear) (Variables|Default|GetExplosionForceFeedbackFar) (Transformation|GetActorLocation) 4.5e+02 7e+02 "Explosion"))
            (:then_3
              (if (Variables|Default|GetFinalBounce)
                (Collision|SetActorEnableCollision false)
                (Development|SetHiddeninGame (Variables|B_Prj_Grenade|GetMesh) true false)
                (Audio|Components|Audio|FadeOut (Variables|B_Prj_Grenade|GetGrenadeFuse_AudioComponent) 0.3 0.0 "Linear" :!class /Script/Engine.AudioComponent)
                (Utilities|Time|ClearandInvalidateTimerbyHandle (Variables|Default|GetFuseIndicatorTimer))
                (Components|Activation|Deactivate (Variables|Default|GetNS_Fuse_Particle))))))))))

(event EventDestroyed)

(event OnEnterWaterEvent(WaterInteractionComponent)
  (params (WaterBody :type Fort%20Water%20Body%20Actor%20Object%20Reference) (WaterInteractionComponent :type Fort%20Water%20Interaction%20Component%20Object%20Reference) (bIsFirstBody :type Boolean))
  (bind _currentnumberofbounces (Variables|Default|GetCurrentNumberOfBounces))
  (bind _effect_distance (Variables|B_Prj_Grenade|GetEffect_Distance))
  ;[1 1 1] OnBounce - Do an explosion
  (Variables|Default|SetE_Surface_Type (Variables|/Script/PhysicsCore.PhysicalMaterial|GetSurfaceType (.PhysMat (Collision|BreakHitResult (stale EventOnBounce Hit)))))
  (bind _bouncelimitreached (CallFunction|HasReachedBounceLimit))
  (if (not _bouncelimitreached)
    (Networking|SwitchHasAuthority
      (:Authority
        (Math|Integer|IncrementInt _currentnumberofbounces
          (:%20%20
            (if (not (> _currentnumberofbounces 1))
              (Variables|Default|SetInititalBounceLocation (.Location (Collision|BreakHitResult (stale EventOnBounce Hit)))))
            (bind _new_velocity (CallFunction|CalculateDirection (.ImpactNormal (Collision|BreakHitResult (stale EventOnBounce Hit)))))
            (Variables|/Script/Engine.MovementComponent|SetVelocity _new_velocity (Variables|Movement|GetProjectileMovementComponent)))))
      (:Remote
        (Utilities|IsValid (Variables|Sounds|GetCue_Bounce)
          (:Is%20Valid
            (Audio|OnCollision (Variables|Default|GetFortCollisionAudio) (.ImpactNormal (Collision|BreakHitResult (stale EventOnBounce Hit))) (* (.ImpactPoint (Collision|BreakHitResult (stale EventOnBounce Hit))) (Variables|/Script/Engine.MovementComponent|GetVelocity (Variables|Movement|GetProjectileMovementComponent)))))))))
  (Transformation|SetWorldRotation _effect_distance "0, 0, 0" false false)
  (Components|Activation|Activate _effect_distance false)
  (Effects|Components|ParticleSystem|SetVectorParameter _effect_distance "Size" (Math|Conversions|ToVector(Float) (Variables|Default|GetExplosionRadius)))
  (if (not (Team|OnSameTeam (Game|GetInstigatorController) (Game|GetPlayerController 0)))
    (Effects|Components|ParticleSystem|SetVectorParameter _effect_distance "Color" "1.500000,0.110000,0.190000"))
  (Delivery|DoExplosion))

(event EventOnBounce
  (params (Hit :type Hit%20Result%20Structure%20%28by%20ref%29))
  (bind (_bblockinghit _binitialoverlap _time _distance _location _impactpoint _normal _impactnormal _physmat) (Collision|BreakHitResult Hit))
  (bind _currentnumberofbounces (Variables|Default|GetCurrentNumberOfBounces))
  (bind _effect_distance (Variables|B_Prj_Grenade|GetEffect_Distance))
  ;[1 1 1] OnBounce - Do an explosion
  (Variables|Default|SetE_Surface_Type (Variables|/Script/PhysicsCore.PhysicalMaterial|GetSurfaceType _physmat))
  (bind _bouncelimitreached (CallFunction|HasReachedBounceLimit))
  (if (not _bouncelimitreached)
    (Networking|SwitchHasAuthority
      (:Authority
        (Math|Integer|IncrementInt _currentnumberofbounces
          (:%20%20
            (if (not (> _currentnumberofbounces 1))
              (Variables|Default|SetInititalBounceLocation _location))
            (bind _new_velocity (CallFunction|CalculateDirection _impactnormal))
            (Variables|/Script/Engine.MovementComponent|SetVelocity _new_velocity (Variables|Movement|GetProjectileMovementComponent)))))
      (:Remote
        (Utilities|IsValid (Variables|Sounds|GetCue_Bounce)
          (:Is%20Valid
            (Audio|OnCollision (Variables|Default|GetFortCollisionAudio) _impactnormal (* _impactpoint (Variables|/Script/Engine.MovementComponent|GetVelocity (Variables|Movement|GetProjectileMovementComponent)))))))))
  (Transformation|SetWorldRotation _effect_distance "0, 0, 0" false false)
  (Components|Activation|Activate _effect_distance false)
  (Effects|Components|ParticleSystem|SetVectorParameter _effect_distance "Size" (Math|Conversions|ToVector(Float) (Variables|Default|GetExplosionRadius)))
  (if (not (Team|OnSameTeam (Game|GetInstigatorController) (Game|GetPlayerController 0)))
    (Effects|Components|ParticleSystem|SetVectorParameter _effect_distance "Color" "1.500000,0.110000,0.190000"))
  (Delivery|DoExplosion))

(event OnExitWaterEvent(WaterInteractionComponent)
  (params (WaterBody :type Fort%20Water%20Body%20Actor%20Object%20Reference) (WaterInteractionComponent :type Fort%20Water%20Interaction%20Component%20Object%20Reference) (bIsLastBody :type Boolean)))

(fn CalculateDirection
  (params (HitNormal :type Vector))
  (bind _projectilemovementcomponent (Variables|Movement|GetProjectileMovementComponent))
  (Variables|/Script/Engine.ProjectileMovementComponent|SetProjectileGravityScale 0.7 _projectilemovementcomponent)
  (bind _output_get (LocalVariables|SetCurrentDirection (Math|Vector|Normalize2D(Vector) (Variables|/Script/Engine.MovementComponent|GetVelocity _projectilemovementcomponent))))
  (Utilities|FlowControl|Sequence
    (:then_0
      (if (> (Variables|Default|GetCurrentNumberOfBounces) 1)
        (bind _output_get_2 (LocalVariables|SetTargetBounceLocation (+ (Variables|Default|GetInititalBounceLocation) (* (Math|Random|RandomUnitVector) (Variables|Default|GetBounceRadius)))))
        (LocalVariables|SetNewVelocity (Math|Vector|MakeVector (.x (* (Math|Vector|Normalize2D(Vector) (- _output_get_2 (Transformation|GetActorLocation))) (Variables|Default|GetBounce%20Force))) (.y (* (Math|Vector|Normalize2D(Vector) (- _output_get_2 (Transformation|GetActorLocation))) (Variables|Default|GetBounce%20Force))) (Variables|Default|GetBounceHeight)))
        (else
          (LocalVariables|SetNewVelocity (Math|Vector|MakeVector (.x (* (Math|Vector|RotateVectorAroundAxis _output_get (Math|Random|RandomFloatinRange (* (- 1.8e+02 (Math|Float|Absolute(Float) (Variables|Default|GetFirstBounceDeadzoneHalfAngle))) -1.0) (- 1.8e+02 (Math|Float|Absolute(Float) (Variables|Default|GetFirstBounceDeadzoneHalfAngle)))) HitNormal) (Variables|Default|GetBounce%20Force))) (.y (* (Math|Vector|RotateVectorAroundAxis _output_get (Math|Random|RandomFloatinRange (* (- 1.8e+02 (Math|Float|Absolute(Float) (Variables|Default|GetFirstBounceDeadzoneHalfAngle))) -1.0) (- 1.8e+02 (Math|Float|Absolute(Float) (Variables|Default|GetFirstBounceDeadzoneHalfAngle)))) HitNormal) (Variables|Default|GetBounce%20Force))) (Variables|Default|GetBounceHeight))))))
    (:then_1
      (return (LocalVariables|GetNewVelocity)))))

(fn CalculateRotation
  (params (NewParam_X :type Float%20%28double-precision%29) (NewParam_Y :type Float%20%28double-precision%29) (NewParam_Z :type Float%20%28double-precision%29))
  (bind (_returnvalue_x _returnvalue_y _returnvalue_z) (Math|Vector|BreakVector (Math|Vector|Normalize (Math|Vector|MakeVector NewParam_X NewParam_Y) 0.0001)))
  (return (Math|Rotator|MakeRotator _returnvalue_x _returnvalue_y _returnvalue_z)))

(fn CalculateVelocity)

(fn DrawDebugDirectionalArrow
  (params (End%20of%20Line :type Vector) (Debug%20Color%20Line :type Linear%20Color%20Structure :default "(R=1.000000,G=0.000000,B=0.067432,A=1.000000)"))
  (if (and (Variables|Default|GetDrawDebug) (Editor|InEditororPIE))
    (Rendering|Debug|DrawDebugArrow (Transformation|GetActorLocation) (+ (Transformation|GetActorLocation) (* End%20of%20Line 4e+02)) 0.0 Debug%20Color%20Line 1e+01 5.0)))

(fn HasReachedBounceLimit
  (Networking|SwitchHasAuthority
    (:Authority
      (Variables|Default|SetFinalBounce (> (Variables|Default|GetCurrentNumberOfBounces) (- (Variables|Default|GetNumberOfBouncesTillExplode) 1))))
    (:Remote
      _))
  (return (Variables|Default|GetFinalBounce)))

(fn OnRep_FinalBounce)

(fn ConstructionScript
  (|CallParentFunction))
