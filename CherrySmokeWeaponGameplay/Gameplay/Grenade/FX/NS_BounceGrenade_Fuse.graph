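# Niagara system for the bounce grenade's burning fuse, combining sparks, embers, a glow and a trail.
# Fuse001 sprays short-lived sparks shaped by wind, curl noise and inherited source movement, while a GPU emitter adds hot embers that fall and cool.
# A persistent sprite and light flicker through waveforms that modulate glow size and color.
# A distance-spawned ribbon leaves a widening trail fading from orange to dark smoke; the separate "fuse" emitter is disabled.
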NiagaraSystem /CherrySmokeWeaponGameplay/Gameplay/Grenade/FX/NS_BounceGrenade_Fuse
dependencies {
 usedDataInterfaces [&NiagaraDataInterfaceCurve, &NiagaraDataInterfaceVectorCurve, &NiagaraDataInterfaceColorCurve]
 usedDynamicInputs []
 usedModules [&SystemState]
}
emitters {
 Fuse001 {
  emitterSpawnScript [/SetVariables(
    Emitter.CurrOri:Quat4f=&FindQuatBetweenTwoVectors(From Vector=<0 0 1>, To Vector=$Engine.Owner.SystemZAxis)
    Emitter.PrevOri:Quat4f=&FindQuatBetweenTwoVectors(From Vector=<0 0 1>, To Vector=$Engine.Owner.SystemZAxis)
   ) {enabled=false, isSetParameters=true}]
  emitterUpdateScript [
   &EmitterState()
   /SetVariables(Emitter.CurrOri:Quat4f=&FindQuatBetweenTwoVectors(From Vector=<0 0 1>, To Vector=$Engine.Owner.SystemZAxis)) {isSetParameters=true}
   &SpawnRate(SpawnRate=55, Use Spawn Probability=false)
   &InterpolateOverTime(Rate Of Change=15, Target Quat Value:Quat4f=$Emitter.CurrOri, Variable Type=::NewEnumerator5)
   /SetVariables(Emitter.PrevOri:Quat4f=$Emitter.CurrOri) {enabled=false, isSetParameters=true}
  ]
  particleSpawnScript [
   &InitializeParticle(
    Lifetime Max=0.2
    Lifetime Min=0.15
    Lifetime Mode=::NewEnumerator1
    Mass Mode=::NewEnumerator3
    Position Offset=<0 0 12>
    Position Offset Coordinate Space=::Local
    Sprite Rotation Mode=::NewEnumerator2
    Sprite Size=<4 8>
    Sprite Size Mode=::NewEnumerator4
    UsePositionOffset=true
    Write Material Random=true
   )
   &ShapeLocation(MaskSpawnGroup=false, Override Seed=false, Sphere Radius=1, UseOwnerScale=true)
   &AddVelocity(
    Cone Angle=35
    Cone Axis=<0 0 1>
    Distribution Along Cone Axis=1
    Inner Cone Angle=15
    Module.UseConeFalloff=true
    Module.UseVelDistribution=true
    Override Seed=false
    Speed Falloff From Cone Axis=1
    Velocity Mode=::NewEnumerator2
    Velocity Speed=&RandomRangeFloat(Minimum=50, Maximum=300)
   )
  ]
  particleUpdateScript [
   &ParticleState()
   &GravityForce()
   &WindForce(Wind Speed=<35 0 0>, Wind Speed Scale=5)
   &Drag(Drag=&RandomRangeFloat(Minimum=0, Maximum=5), Use Linear Drag=true, Use Rotational Drag=false)
   &ScaleColor(
    Scale Alpha=&FloatFromCurve(FloatCurve=DataInterface() {propertyValues {Curve {keys [<at -0.0046 1.0113 RCIM_Cubic>,<at 1.0011 0.002 RCIM_Cubic>]}}})
    ScaleA=true
    ScaleRGB=true
   ) {enabled=false}
   &InheritVelocity(Speed Limit=false, UseSpeedThreshold=false) {enabled=false}
   &CurlNoiseForce(Bias Noise=false, Noise Frequency=44, Noise Strength=555, Pan Noise=false)
   &InheritSourceMovement(
    Applied Position Scale=&VectorFromFloat(Value=&FloatFromCurve(FloatCurve=DataInterface() {propertyValues {Curve {
         defaultValue=1
         keys [
          {tangentMode=RCTM_User, time=0, value=1}
          {arriveTangent=-1, interpMode=RCIM_Cubic, leaveTangent=-1, time=0.5, value=1}
          {interpMode=RCIM_Cubic, leaveTangent=-8.0089, time=1, value=0}
         ]
        }}}))
    ApplyForce=false
    ApplyPosition=true
    ApplyPositionFalloff=false
   )
   &SolveForcesAndVelocity(Clamp Velocity=false, Limit Acceleration=false)
   &ScaleSpriteSizeBySpeed(Max Scale Factor=<1 3>, Min Scale Factor=<0.3 0.5>)
   &ScaleSpriteSize(
    Initial Sprite Size:Vector2f=$StackContext.SpriteSize
    Scale Factor=&MakeVector2D(
     X=1
     Y=&FloatFromCurve(FloatCurve=DataInterface() {propertyValues {Curve {
         defaultValue=1
         keys [
          {tangentMode=RCTM_User, time=0, value=1}
          {arriveTangent=-0.0405, leaveTangent=-0.0405, time=0.5, value=1}
          {arriveTangent=-8.0089, leaveTangent=-8.0089, time=1, value=0}
         ]
        }}})
    )
    Scale Sprite Size Mode=::NewEnumerator2
   )
   &Color(Color=&MakeLinearColorFromVectorAndFloat(
     "Vector (RGB)"=<238.589 81.2308 24.9103>
     "Float (Alpha)"=&FloatFromCurve(FloatCurve=DataInterface() {propertyValues {Curve {keys [<at 0 0 RCIM_Cubic>,<at 0.0967 0.1386 RCIM_Cubic>,<at 1 0>]}}})
    ))
  ]
  rendererClasses [&NiagaraSpriteRendererProperties]
  renderers [{rendererClass=&NiagaraSpriteRendererProperties}]
 }
 GPU_up_Delay {
  simTarget=GPUComputeSim
  emitterUpdateScript [
   &EmitterState()
   &SpawnBurst_Instantaneous(Spawn Count=0, Use Spawn Probability=false, UseLoopCountLimit=false)
   &SpawnRate(SpawnRate=50, Use Spawn Probability=false)
  ]
  particleSpawnScript [
   &InitializeParticle(
    Lifetime=&RandomRangeFloat(Minimum=0.2, Maximum=0.5)
    Sprite Rotation Angle=&RandomRangeFloat(Minimum=0, Maximum=1)
    Sprite Rotation Mode=::NewEnumerator1
    Sprite Size=&Vector2DFromFloat(Value=&RandomRangeFloat(Minimum=1, Maximum=3))
    Sprite Size Mode=::NewEnumerator4
    UsePositionOffset=false
    Write Material Random=true
   )
   &AddVelocity(Velocity=&RandomRangeVector(Minimum=<30 30 30>, Maximum=<-30 -30 400>))
   &ShapeLocation(
    Cylinder Height=2
    Cylinder Radius=20
    MaskSpawnGroup=false
    Override Seed=false
    Shape Primitive=::NewEnumerator1
    Surface Only=false
    UseOwnerScale=true
   )
  ]
  particleUpdateScript [
   &ParticleState()
   &ScaleSpriteSize(
    Scale Sprite Size Mode=::NewEnumerator0
    Uniform Scale Factor=&FloatFromCurve(FloatCurve=DataInterface() {propertyValues {Curve {keys [
         <at -0.0044 0.2654>
         {arriveTangent=0.9116, interpMode=RCIM_Cubic, leaveTangent=0.9116, time=0.24, value=1.0117}
         {arriveTangent=-1.3312, interpMode=RCIM_Cubic, leaveTangent=-1.3312, time=0.8015, value=1}
         <at 1 0>
        ]}}})
   )
   &AccelerationForce(Acceleration=<0 0 -2200>)
   &SolveForcesAndVelocity(Clamp Velocity=false, Limit Acceleration=false)
   &Color(Color=&MakeLinearColorFromVectorAndFloat(
     "Vector (RGB)"=&VectorFromCurve(VectorCurve=DataInterface() {propertyValues {
        XCurve {keys [
          <at 0.0095 5107.2236 RCIM_Cubic>
          {arriveTangent=-6404.0493, interpMode=RCIM_Cubic, leaveTangent=-6404.0493, time=0.0697, value=5002.8364}
          <at 0.806 6.1037>
         ]}
        YCurve {keys [
          <at 0.0095 776.5526 RCIM_Cubic>
          {arriveTangent=-974.8029, interpMode=RCIM_Cubic, leaveTangent=-974.8029, time=0.0697, value=760.663}
          <at 0.806 0.0772>
         ]}
        ZCurve {keys [
          <at 0.0095 54.4699 RCIM_Cubic>
          {arriveTangent=-59.9813, interpMode=RCIM_Cubic, leaveTangent=-59.9813, time=0.0697, value=181.8841}
          <at 0.806 6.692>
         ]}
       }})
     "Float (Alpha)"=&ClampFloat(Float=&FloatFromCurve(FloatCurve=DataInterface() {propertyValues {Curve {keys [<at 0.8224 1>,<at 1 0>]}}}))
    ))
  ]
  rendererClasses [&NiagaraSpriteRendererProperties]
  renderers [{rendererClass=&NiagaraSpriteRendererProperties}]
 }
 Particle_Emitter {
  emitterUpdateScript [
   &EmitterState()
   &SpawnBurst_Instantaneous(Use Spawn Probability=false, UseLoopCountLimit=false)
  ]
  particleSpawnScript [&InitializeParticle(Lifetime=10, Sprite Size=<100 100>, Sprite Size Mode=::NewEnumerator4, UsePositionOffset=false, Write Material Random=true)]
  particleUpdateScript [
   /SetVariables(Particles.LightExponent=5) {isSetParameters=true}
   &ParticleState(Kill Particles When Lifetime Has Elapsed=false, Let Infinitely Lived Particles Die When Emitter Deactivates=true)
   &ScaleSpriteSize(
    Scale Sprite Size Mode=::NewEnumerator0
    Uniform Scale Factor=&Waveform("Waveform [1]"=::NewEnumerator12, "[1] Frequency"=10, "[1] Amplitude Min/Max"=<1 2>)
   )
   &SpriteRotationRate(Rotation Rate=500) {enabled=false}
   &Color(Color=&Multiply_LinearColorByFloat(
     Linear Color=&ColorFromCurve(ColorCurve=DataInterface() {propertyValues {
        AlphaCurve {keys [<at 0 0.1 RCIM_Cubic>,<at 1 0 RCIM_Cubic>]}
        BlueCurve {keys [<at 0 0.0096 RCIM_Cubic>]}
        GreenCurve {keys [<at 0 1.3208 RCIM_Cubic>]}
        RedCurve {keys [<at 0 5 RCIM_Cubic>]}
       }})
     Float=&Waveform("Waveform [1]"=::NewEnumerator12, "[1] Frequency"=10, "[1] Amplitude Min/Max"=<0.5 2>)
    ))
   &CascadeConversion_LightProperties(LightBrightnessScale=5, LightColorScale=<1 1 1>, LightRadiusScale=1) {enabled=false}
   &FN_Light_Attributes(
    Color=&Multiply_LinearColorByFloat(Linear Color=$StackContext.Color, Float=1)
    Create Color Attribute=true
    Create Radius Attribute=true
    Radius=200
    Scale Color=3
   )
   &CameraOffset(Camera Offset Amount=-20)
  ]
  rendererClasses [&NiagaraSpriteRendererProperties, &NiagaraLightRendererProperties]
  renderers [
   {rendererClass=&NiagaraSpriteRendererProperties}
   {rendererClass=&NiagaraLightRendererProperties, rendererIndex=1}
  ]
 }
 Ribbon {
  emitterSpawnScript [/SetVariables(Emitter.Static bool=true) {isSetParameters=true}]
  emitterUpdateScript [
   &EmitterState()
   &SpawnPerUnit(Module.Use Max Movement Threshold=false, Module.Use Movement Tolerance=false, Use Spawn Probability=false)
  ]
  particleSpawnScript [
   /SetVariables() {isSetParameters=true}
   &InitializeRibbon(
    Color=<rgba 0.7135 0.7135 0.7135 1>
    Lifetime=0.8
    Ribbon Width=&Multiply_Float(A=15, B=2)
    UsePositionOffset=false
    Write Material Random=true
   )
   &DynamicMaterialParameters(Module.Index0Param0Write=true, Module.Index0Param1Write=false, Module.Index0Param2Write=false, Module.Index0Param3Write=false)
  ]
  particleUpdateScript [
   /SetVariables(Particles.ScaledRibbonWidth:NiagaraFloat=&Multiply_Float(
     A=$Particles.RibbonWidth
     B=&FloatFromCurve(FloatCurve=DataInterface() {propertyValues {Curve {keys [
          {arriveTangent=5.3994, interpMode=RCIM_Cubic, leaveTangent=5.3994, tangentMode=RCTM_User, time=0, value=0.5}
          {interpMode=RCIM_Cubic, leaveTangent=5.8784, time=1, value=2.5}
         ]}}})
    )) {isSetParameters=true}
   &ParticleState()
   &Color(Color=&ColorFromCurve(ColorCurve=DataInterface() {propertyValues {
       AlphaCurve {keys [<at 0.0479 1 RCIM_Cubic>,<at 0.8514 0 RCIM_Cubic>]}
       BlueCurve {keys [<at 0 0 RCIM_Cubic>,<at 0.073 0.033>]}
       GreenCurve {keys [<at 0 0.1804 RCIM_Cubic>,<at 0.073 0.0341>]}
       RedCurve {keys [<at 0 1 RCIM_Cubic>,<at 0.073 0.0365>]}
      }}))
   &AccelerationForce(Acceleration=$Particles.Acceleration)
   &SolveForcesAndVelocity(Clamp Velocity=false, Limit Acceleration=false)
   &DynamicMaterialParameters(
    Index 0 Param 1=&FloatFromCurve(FloatCurve=DataInterface() {propertyValues {Curve {defaultValue=1, keys [{tangentMode=RCTM_User, time=0, value=1}, {tangentMode=RCTM_User, time=1, value=0}]}}})
    Index 0 Param 2=&FloatFromCurve(FloatCurve=DataInterface() {propertyValues {Curve {keys [<at 0.5 1>,<at 1 0>]}}})
    Index 0 Param 3=-15
    Index 0 Param 4=1
    Module.Index0Float0Write=true
    Module.Index0Float1Write=true
    Module.Index0Float2Write=true
    Module.Index0Float3Write=true
   ) {module=DynamicMaterialParameters001}
  ]
  rendererClasses [&NiagaraRibbonRendererProperties]
  renderers [{rendererClass=&NiagaraRibbonRendererProperties}]
 }
 fuse {
  bEnabled=false
  emitterUpdateScript [
   &EmitterState(Inactive Response=::NewEnumerator1, Life Cycle Mode=::NewEnumerator1, Loop Duration=1, UseLoopDelay=true)
   &SpawnBurst_Instantaneous(Use Spawn Probability=false, UseLoopCountLimit=false)
   &SpawnRate(SpawnRate=10, Use Spawn Probability=false)
  ]
  particleSpawnScript [
   /SetVariables(Particles.SpriteRotationRate=&RandomRangeFloat(Minimum=0, Maximum=0.01)) {isSetParameters=true}
   &InitializeParticle(
    Lifetime=&RandomRangeFloat(Minimum=0.1, Maximum=0.2)
    Sprite Rotation Angle=&RandomRangeFloat(Minimum=0, Maximum=1)
    Sprite Rotation Mode=::NewEnumerator1
    Sprite Size=&RandomRangeVector2D(Minimum=<30 5>, Maximum=<200 35>)
    Sprite Size Mode=::NewEnumerator4
    UsePositionOffset=false
    Write Material Random=true
   )
   &CameraOffset(Camera Offset Amount=&RandomRangeFloat(Minimum=10, Maximum=60))
  ]
  particleUpdateScript [
   &ParticleState()
   &SpriteRotationRate(Rotation Rate=&AngleConversion(
     Angle:NiagaraFloat=$Particles.SpriteRotationRate
     Angle Input:ENiagara_AngleInput=::NewEnumerator1
     Angle Output:ENiagara_AngleInput=::NewEnumerator0
    ))
   &ScaleSpriteSize(
    Scale Factor=&Vector2DFromFloat(Value=&FloatFromCurve(FloatCurve=DataInterface() {propertyValues {Curve {keys [
          <at -0.0011 0.0573>
          {arriveTangent=-0.0322, interpMode=RCIM_Cubic, leaveTangent=-0.0322, time=0.0867, value=0.7609}
          {interpMode=RCIM_Cubic, leaveTangent=0.0935, time=1.0043, value=0.0249}
         ]}}}))
    Scale Sprite Size Mode=::NewEnumerator2
   )
   &Color(Color=&MakeLinearColorFromVectorAndFloat(
     "Vector (RGB)"=<238.589 81.2308 24.9103>
     "Float (Alpha)"=&FloatFromCurve(FloatCurve=DataInterface() {propertyValues {Curve {keys [<at 0 0 RCIM_Cubic>,<at 0.0967 0.1386 RCIM_Cubic>,<at 1 0>]}}})
    ))
  ]
  rendererClasses [&NiagaraSpriteRendererProperties]
  renderers [{rendererClass=&NiagaraSpriteRendererProperties}]
 }
}
summary {
 systemName=NS_BounceGrenade_Fuse
 emitters [
  {bEnabled=false, emitterName=fuse, rendererClasses [&NiagaraSpriteRendererProperties]}
  {emitterName=GPU_up_Delay, simTarget=GPUComputeSim, rendererClasses [&NiagaraSpriteRendererProperties]}
  {emitterName=Ribbon, rendererClasses [&NiagaraRibbonRendererProperties]}
  {
   emitterName=Particle_Emitter
   rendererClasses [&NiagaraSpriteRendererProperties, &NiagaraLightRendererProperties]
  }
  {emitterName=Fuse001, rendererClasses [&NiagaraSpriteRendererProperties]}
 ]
 userVariables []
}
