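# Default Lit material for the exotic grenade, with a galaxy effect and a two-color outline.
# MF_Effect_Galaxy runs at full mask on attributes holding only the Mark II emissive texture, then feeds MF_Effect_TOD.
# That emissive texture is boosted and added back, then an outline mask blends a pre-skinned radial gradient of squared CenterYellow and CenterPurple into base color and emissive.
# The result passes through MF_TechArtOverrides_accessories.
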Material /CherrySmokeWeaponGameplay/Gameplay/Grenade/Material/M_GrenadeExotic
graph {
 properties {
  bAutomaticallySetUsageInEditor=false
  bUseMaterialAttributes=true
  bUsedWithMeshDeformer=true
  bUsedWithSkeletalMesh=true
  usedShadingModels=Default Lit
 }
 nodes [
  n1 = TextureSample() {texture=/Game/Gadgets/Assets/Grenade_Mark_II/Textures/T_Grenade_MK_II_E}
  n2 = MakeMaterialAttributes!(
   BaseColor
   Metallic
   Specular
   Roughness
   Anisotropy
   EmissiveColor=@n1.RGB
   Opacity
   OpacityMask
   Normal
   Tangent
   WorldPositionOffset
   SubsurfaceColor
   ClearCoat
   ClearCoatRoughness
   AmbientOcclusion
   Refraction
   CustomizedUVs_0
   CustomizedUVs_1
   CustomizedUVs_2
   CustomizedUVs_3
   CustomizedUVs_4
   CustomizedUVs_5
   CustomizedUVs_6
   CustomizedUVs_7
   PixelDepthOffset
   ShadingModel
   Displacement
  )
  n3 = Constant() {r=1}
  n4 = VertexInterpolator(VS=PreSkinnedNormal())
  n5 = GetMaterialAttributes(None=MaterialFunctionCall!(Material=MaterialFunctionCall!(
     Material=@n2
     IsState
     IsState_B
     CM_Msk=@n3
     Pre-Skinned Local Normal=@n4.PS
     Calcated_Bounds
     Object-SafeZone-VP-Character-PreView-Switch
     Object-SafeZone-Virtual-Pivot
    ) {materialFunction=/Game/Weapons/WeaponSkins/Wraps/MaterialFunctions/MF_Effect_Galaxy}.Result) {materialFunction=/Game/Weapons/WeaponSkins/Wraps/MaterialFunctions/MF_Effect_TOD}.Result) {attributeGetTypes [#1, #2]}
  n6 = VectorParameter() {parameterName=CenterYellow, defaultValue=<rgba 0 0 0 1>}
  n7 = VectorParameter() {parameterName=CenterPurple, defaultValue=<rgba 0 0 0 1>}
  n8 = PreSkinnedPosition()
  n9 = Divide(
   A=Length(None=VertexInterpolator(VS=@n8).PS) {is=class MaterialExpressionLength}
   B=ScalarParameter() {defaultValue=20, parameterName=GradientRadius_OUTLine}
  )
  n10 = LinearInterpolate(A=Multiply(A=@n6.RGB, B=@n6.RGB), B=Multiply(A=@n7.RGB, B=@n7.RGB), Alpha=SmoothStep(Value=Saturate(None=@n9)) {constMin=0.5})
  n11 = TextureSample() {samplerType=SAMPLERTYPE_Masks, texture=/CherrySmokeWeaponGameplay/Gameplay/Grenade/Texture/T_Grenade_OUTline_Exotic}
  n12 = Add(A=@n5.Emissive Color, B=Multiply(A=@n1.RGB) {constB=3})
  MaterialFunctionCall(In=SetMaterialAttributes!(
    MaterialAttributes=@n5.MaterialAttributes
    Base Color=LinearInterpolate(A=@n5.Base Color, B=@n10, Alpha=@n11.RGB)
    Emissive Color=LinearInterpolate(A=@n12, B=@n10, Alpha=@n11.RGB)
   ) {attributeSetTypes [#1, #2]}) {materialFunction=&MF_TechArtOverrides_accessories}
 ]
}
