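# Sound cue for the bounce sound of the CherrySmoke unstable grenade.
# A random node picks one of three mono bounce waves with equal weight, passed through a modulator with pitch fixed at normal.
# It uses the grenade bounce attenuation and the ranged weapon projectile effects sound class at reduced volume.
# A concurrency override caps simultaneous plays and blocks new ones once the limit is reached.
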SoundCue /CherrySmokeWeaponGameplay/Gameplay/Grenade/Sounds/SC_CherrySmoke_Athena_Grenade_Unstable_Bounce
root=Modulator_5
nodes [
 {id=Modulator_5, children [Random_0], properties {pitchMax=1, pitchMin=1}}
 {id=Random_0, children [WavePlayer_0,WavePlayer_1,WavePlayer_2], properties {weights [1,1,1]}}
 {id=WavePlayer_0, properties {soundWaveAssetPtr=&/Grenade_Bounce_Mono_01}}
 {id=WavePlayer_1, properties {soundWaveAssetPtr=&/Grenade_Bounce_Mono_02}}
 {id=WavePlayer_2, properties {soundWaveAssetPtr=&/Grenade_Bounce_Mono_03}}
]
properties {
 attenuationSettings=/Game/Sounds/Attenuation/Fort_Grenade_Bounce_Att
 bOverrideConcurrency=true
 maxDistance=5500
 soundClassObject=&Fort_WeaponRanged_Projectile_Effects
 volumeMultiplier=0.65
 concurrencyOverrides {maxCount=3, resolutionRule=PreventNew}
}
