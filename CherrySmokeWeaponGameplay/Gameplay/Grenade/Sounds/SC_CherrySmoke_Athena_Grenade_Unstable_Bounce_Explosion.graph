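# Sound cue for the CherrySmoke unstable grenade's bounce explosion, layering a bounce impact with explosion audio.
# The root mixer combines a random bounce sample with a sub-mix of close, medium and distant explosion layers, each randomized and using its own attenuation and sound class.
# The close layer starts after a short fixed delay, while medium and distant layers are mixed progressively quieter.
# Overridden concurrency limits simultaneous instances.
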SoundCue /CherrySmokeWeaponGameplay/Gameplay/Grenade/Sounds/SC_CherrySmoke_Athena_Grenade_Unstable_Bounce_Explosion
root=Mixer_0
nodes [
 {
  id=Attenuation_0
  children [Modulator_1]
  properties {
   attenuationSettings=/Game/Sounds/Attenuation/Fort_Explosion_Distant_Athena_Att
   attenuationOverrides {falloffDistance=3000, attenuationShapeExtents {x=1000}}
  }
 }
 {id=Attenuation_1, children [Modulator_2], properties {attenuationSettings=/Game/Sounds/Attenuation/Fort_Grenade_Bounce_Att}}
 {
  id=Attenuation_4
  children [Random_5]
  properties {attenuationSettings=/Game/Sounds/Attenuation/Fort_Explosion_Close_Athena_Att, attenuationOverrides {falloffDistance=1600}}
 }
 {
  id=Attenuation_5
  children [Modulator_5]
  properties {
   attenuationSettings=/Game/Sounds/Attenuation/Fort_Explosion_Medium_Athena_Att
   attenuationOverrides {falloffDistance=3000, attenuationShapeExtents {x=1000}}
  }
 }
 {id=Delay_0, children [SoundClass_3], properties {delayMax=0.06, delayMin=0.06}}
 {id=Mixer_0, children [Mixer_4,SoundClass_1], properties {inputVolume [1,1]}}
 {id=Mixer_4, children [Delay_0,SoundClass_2,SoundClass_0], properties {inputVolume [1,0.8,0.6]}}
 {id=Modulator_0, children [Attenuation_4]}
 {id=Modulator_1, children [Random_3], properties {pitchMax=0.95, pitchMin=0.85, volumeMax=1.1, volumeMin=1}}
 {id=Modulator_2, children [Random_0]}
 {id=Modulator_5, children [Random_4], properties {pitchMax=1.1, pitchMin=0.9, volumeMax=1, volumeMin=0.9}}
 {id=Random_0, children [WavePlayer_0,WavePlayer_1,WavePlayer_2], properties {weights [1,1,1]}}
 {id=Random_3, children [WavePlayer_20, WavePlayer_21, WavePlayer_22, WavePlayer_23], properties {weights [1,1,1,1]}}
 {id=Random_4, children [WavePlayer_16, WavePlayer_17, WavePlayer_18, WavePlayer_19], properties {weights [1,1,1,1]}}
 {id=Random_5, children [WavePlayer_12, WavePlayer_13, WavePlayer_14, WavePlayer_15], properties {weights [1,1,1,1]}}
 {
  id=SoundClass_0
  children [Attenuation_0]
  properties {soundClassOverride=&Fort_WeaponRanged_Fire_Distant}
 }
 {
  id=SoundClass_1
  children [Attenuation_1]
  properties {soundClassOverride=&Fort_WeaponRanged_Projectile_Effects}
 }
 {
  id=SoundClass_2
  children [Attenuation_5]
  properties {soundClassOverride=&Fort_WeaponRanged_Fire_Distant}
 }
 {id=SoundClass_3, children [Modulator_0], properties {soundClassOverride=&Fort_WeaponRanged_Fire_3P_PSM}}
 {id=WavePlayer_0, properties {soundWaveAssetPtr=&/Grenade_Bounce_Mono_01}}
 {id=WavePlayer_1, properties {soundWaveAssetPtr=&/Grenade_Bounce_Mono_02}}
 {
  id=WavePlayer_12
  properties {soundWaveAssetPtr=&/CherrySmoke_Athena_Grenade_Unstable_Bounce_Close_01}
 }
 {
  id=WavePlayer_13
  properties {soundWaveAssetPtr=&/CherrySmoke_Athena_Grenade_Unstable_Bounce_Close_02}
 }
 {
  id=WavePlayer_14
  properties {soundWaveAssetPtr=&/CherrySmoke_Athena_Grenade_Unstable_Bounce_Close_03}
 }
 {
  id=WavePlayer_15
  properties {soundWaveAssetPtr=&/CherrySmoke_Athena_Grenade_Unstable_Bounce_Close_04}
 }
 {
  id=WavePlayer_16
  properties {soundWaveAssetPtr=&/CherrySmoke_Athena_Grenade_Unstable_Bounce_Medium_01}
 }
 {
  id=WavePlayer_17
  properties {soundWaveAssetPtr=&/CherrySmoke_Athena_Grenade_Unstable_Bounce_Medium_02}
 }
 {
  id=WavePlayer_18
  properties {soundWaveAssetPtr=&/CherrySmoke_Athena_Grenade_Unstable_Bounce_Medium_03}
 }
 {
  id=WavePlayer_19
  properties {soundWaveAssetPtr=&/CherrySmoke_Athena_Grenade_Unstable_Bounce_Medium_04}
 }
 {id=WavePlayer_2, properties {soundWaveAssetPtr=&/Grenade_Bounce_Mono_03}}
 {
  id=WavePlayer_20
  properties {soundWaveAssetPtr=&/CherrySmoke_Athena_Grenade_Unstable_Bounce_Distant_01}
 }
 {
  id=WavePlayer_21
  properties {soundWaveAssetPtr=&/CherrySmoke_Athena_Grenade_Unstable_Bounce_Distant_02}
 }
 {
  id=WavePlayer_22
  properties {soundWaveAssetPtr=&/CherrySmoke_Athena_Grenade_Unstable_Bounce_Distant_03}
 }
 {
  id=WavePlayer_23
  properties {soundWaveAssetPtr=&/CherrySmoke_Athena_Grenade_Unstable_Bounce_Distant_04}
 }
]
properties {bOverrideConcurrency=true, maxDistance=27500, volumeMultiplier=0.8, concurrencyOverrides {maxCount=3, volumeScale=0.5}}
