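# Sound cue for the CherrySmoke unstable grenade explosion in Athena, layering close, medium and distant variations by listener range.
# Each layer randomly picks one of four waves, with distance attenuation, modulation and a sound class override; close uses the third-person fire class, the others the distant class.
# A mixer sums the layers, weighted toward close, and a concurrency override caps simultaneous instances and scales their volume.
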SoundCue /CherrySmokeWeaponGameplay/Gameplay/Grenade/Sounds/SC_CherrySmoke_Athena_Grenade_Unstable_Explosion
root=Mixer_4
nodes [
 {
  id=Attenuation_0
  children [Modulator_1]
  properties {
   attenuationSettings=/Game/Sounds/Attenuation/Fort_Explosion_Distant_Athena_Att
   attenuationOverrides {falloffDistance=3000, attenuationShapeExtents {x=1000}}
  }
 }
 {
  id=Attenuation_4
  children [Random_2]
  properties {attenuationSettings=/Game/Sounds/Attenuation/Fort_Explosion_Close_Athena_Att, attenuationOverrides {falloffDistance=1600}}
 }
 {
  id=Attenuation_5
  children [Modulator_5]
  properties {
   attenuationSettings=/Game/Sounds/Attenuation/Fort_Explosion_Medium_Athena_Att
   attenuationOverrides {falloffDistance=3000, attenuationShapeExtents {x=1000}}
  }
 }
 {id=Mixer_4, children [SoundClass_3,SoundClass_2,SoundClass_0], properties {inputVolume [1.2,0.8,0.6]}}
 {id=Modulator_0, children [Attenuation_4]}
 {id=Modulator_1, children [Random_0], properties {pitchMax=0.95, pitchMin=0.85, volumeMax=1.1, volumeMin=1}}
 {id=Modulator_5, children [Random_1], properties {pitchMax=1.1, pitchMin=0.9, volumeMax=1, volumeMin=0.9}}
 {id=Random_0, children [WavePlayer_0,WavePlayer_1,WavePlayer_2,WavePlayer_3], properties {weights [1,1,1,1]}}
 {id=Random_1, children [WavePlayer_8, WavePlayer_9, WavePlayer_10, WavePlayer_11], properties {weights [1,1,1,1]}}
 {id=Random_2, children [WavePlayer_4,WavePlayer_5,WavePlayer_6,WavePlayer_7], properties {weights [1,1,1,1]}}
 {
  id=SoundClass_0
  children [Attenuation_0]
  properties {soundClassOverride=&Fort_WeaponRanged_Fire_Distant}
 }
 {
  id=SoundClass_2
  children [Attenuation_5]
  properties {soundClassOverride=&Fort_WeaponRanged_Fire_Distant}
 }
 {id=SoundClass_3, children [Modulator_0], properties {soundClassOverride=&Fort_WeaponRanged_Fire_3P_PSM}}
 {
  id=WavePlayer_0
  properties {soundWaveAssetPtr=&/CherrySmoke_Athena_Grenade_Unstable_Explosion_Distant_01}
 }
 {
  id=WavePlayer_1
  properties {soundWaveAssetPtr=&/CherrySmoke_Athena_Grenade_Unstable_Explosion_Distant_02}
 }
 {
  id=WavePlayer_10
  properties {soundWaveAssetPtr=&/CherrySmoke_Athena_Grenade_Unstable_Explosion_Medium_03}
 }
 {
  id=WavePlayer_11
  properties {soundWaveAssetPtr=&/CherrySmoke_Athena_Grenade_Unstable_Explosion_Medium_04}
 }
 {
  id=WavePlayer_2
  properties {soundWaveAssetPtr=&/CherrySmoke_Athena_Grenade_Unstable_Explosion_Distant_03}
 }
 {
  id=WavePlayer_3
  properties {soundWaveAssetPtr=&/CherrySmoke_Athena_Grenade_Unstable_Explosion_Distant_04}
 }
 {
  id=WavePlayer_4
  properties {soundWaveAssetPtr=&/CherrySmoke_Athena_Grenade_Unstable_Explosion_Close_01}
 }
 {
  id=WavePlayer_5
  properties {soundWaveAssetPtr=&/CherrySmoke_Athena_Grenade_Unstable_Explosion_Close_02}
 }
 {
  id=WavePlayer_6
  properties {soundWaveAssetPtr=&/CherrySmoke_Athena_Grenade_Unstable_Explosion_Close_03}
 }
 {
  id=WavePlayer_7
  properties {soundWaveAssetPtr=&/CherrySmoke_Athena_Grenade_Unstable_Explosion_Close_04}
 }
 {
  id=WavePlayer_8
  properties {soundWaveAssetPtr=&/CherrySmoke_Athena_Grenade_Unstable_Explosion_Medium_01}
 }
 {
  id=WavePlayer_9
  properties {soundWaveAssetPtr=&/CherrySmoke_Athena_Grenade_Unstable_Explosion_Medium_02}
 }
]
properties {bOverrideConcurrency=true, maxDistance=27500, volumeMultiplier=1.2, concurrencyOverrides {maxCount=3, volumeScale=0.5}}
