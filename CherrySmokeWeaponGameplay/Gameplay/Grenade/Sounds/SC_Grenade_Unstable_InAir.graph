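# Sound cue for the in-air sound of the unstable grenade, with concurrency limited per owner.
# Mixes a random frag launch one-shot, routed through a quality-level switch, with a Doppler-shifted loop bed that fades in.
# The bed layers an HRTF-spatialized flight loop with fuse and warning loops under their own attenuation.
# A PreExplo parameter continuously modulates pitch and volume of the fuse and warning loops to signal the coming explosion.
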SoundCue /CherrySmokeWeaponGameplay/Gameplay/Grenade/Sounds/SC_Grenade_Unstable_InAir
root=Mixer_0
nodes [
 {
  id=Attenuation_0
  children [SoundClass_1]
  properties {
   attenuationSettings=/Game/Sounds/Fort_Weapons/Grenades/Frag/Grenade_Frag_InAir_Att
   attenuationOverrides {
    distanceAlgorithm=Logarithmic
    falloffDistance=1536
    spatializationAlgorithm=SPATIALIZATION_HRTF
    stereoSpread=0
    attenuationShapeExtents {x=768}
   }
  }
 }
 {id=Attenuation_2, children [Mixer_2], properties {attenuationSettings=/Game/Sounds/Attenuation/Fort_Explosive_Fuse_Ambient_Att}}
 {
  id=Attenuation_3
  children [WavePlayer_0]
  properties {
   attenuationSettings=&Grenade_Frag_InAir_Att
   attenuationOverrides {
    distanceAlgorithm=Logarithmic
    falloffDistance=1536
    spatializationAlgorithm=SPATIALIZATION_HRTF
    stereoSpread=0
    attenuationShapeExtents {x=768}
   }
  }
 }
 {id=Doppler_0, children [Mixer_1], properties {dopplerIntensity=2}}
 {
  id=Enveloper_0
  children [SoundClass_0]
  properties {pitchCurve {editorCurveData {keys [{time=0}]}}, volumeCurve {editorCurveData {keys [<at 0 0>,<at 0.5 1>]}}}
 }
 {id=Mixer_0, children [QualityLevel_0, Enveloper_0], properties {inputVolume [1.1,1]}}
 {id=Mixer_1, children [Attenuation_3, Attenuation_2], properties {inputVolume [1.25,1]}}
 {id=Mixer_2, children [ModulatorContinuous_0, ModulatorContinuous_1], properties {inputVolume [0.5,2.5]}}
 {
  id=ModulatorContinuous_0
  children [Modulator_0]
  properties {
   pitchModulationParams {default=0, maxOutput=1.5, minOutput=1, parameterName=PreExplo}
   volumeModulationParams {default=0, maxOutput=1.2, minOutput=1, parameterName=PreExplo}
  }
 }
 {
  id=ModulatorContinuous_1
  children [WavePlayer_3]
  properties {
   pitchModulationParams {default=0, maxOutput=1.8, minOutput=1, parameterName=PreExplo}
   volumeModulationParams {default=0, minOutput=0.1, parameterName=PreExplo}
  }
 }
 {id=Modulator_0, children [WavePlayer_1]}
 {id=Modulator_1, children [Random_1]}
 {id=QualityLevel_0, children [Attenuation_0, Attenuation_0, Attenuation_0, null]}
 {id=Random_1, children [WavePlayer_7,WavePlayer_8,WavePlayer_9], properties {bShouldExcludeFromBranchCulling=true, weights [1,1,1]}}
 {id=SoundClass_0, children [Doppler_0], properties {soundClassOverride=/Game/Sounds/FortSoundClassesAndModes/Fort_Explosives_Medium_PSM}}
 {id=SoundClass_1, children [Modulator_1], properties {soundClassOverride=&Fort_Gadgets}}
 {id=WavePlayer_0, properties {bLooping=true, soundWaveAssetPtr=/Game/Sounds/Fort_Weapons/Launchers/Grenade_InAirLoop_Small_01}}
 {id=WavePlayer_1, properties {bLooping=true, soundWaveAssetPtr=/Game/Sounds/Fort_GamePlay_Sounds/Explosives/fuse_m80_loop_shortened}}
 {
  id=WavePlayer_3
  properties {bLooping=true, soundWaveAssetPtr=/Game/Sounds/Fort_GamePlay_Sounds/Explosives/GrenadeLauncher_Fuse_PreExploWarning_01}
 }
 {id=WavePlayer_7, properties {soundWaveAssetPtr=/Game/Sounds/Fort_Weapons/Grenades/Frag/Grenade_Frag_InAir_Start_01}}
 {id=WavePlayer_8, properties {soundWaveAssetPtr=/Game/Sounds/Fort_Weapons/Grenades/Frag/Grenade_Frag_InAir_Start_02}}
 {id=WavePlayer_9, properties {soundWaveAssetPtr=/Game/Sounds/Fort_Weapons/Grenades/Frag/Grenade_Frag_InAir_Start_03}}
]
properties {
 maxDistance=6300
 volumeMultiplier=0.7
 attenuationOverrides {falloffDistance=5000, attenuationShapeExtents {x=1000}}
 concurrencyOverrides {bLimitToOwner=true, maxCount=3, resolutionRule=StopQuietest}
 concurrencySet [/Game/Sounds/Concurrency_Rules/Athena_Explosive_Fuse_Conc]
}
